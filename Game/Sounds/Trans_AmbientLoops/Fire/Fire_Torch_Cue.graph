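# Sound cue for a looping torch fire ambience, set to lowest priority.
# Picks one of three looping fire waves at equal weight, passes it through slight pitch modulation, then an envelope that ramps volume up from silence.
# Uses a shared torch focus-cone attenuation and its own three-voice concurrency limit with a release time on stolen voices.
# Routes to the Fort_Ambient_NonLoop sound class despite the looping waves.
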SoundCue /Game/Sounds/Trans_AmbientLoops/Fire/Fire_Torch_Cue
root=Enveloper_1
nodes [
 {
  id=Enveloper_1
  children [Modulator_0]
  properties {pitchCurve {editorCurveData {keys [{time=0}]}}, volumeCurve {editorCurveData {keys [<at 0 0>,<at 1 1>]}}}
 }
 {id=Modulator_0, children [Random_0], properties {pitchMax=1.1, pitchMin=0.9}}
 {id=Random_0, children [WavePlayer_1,WavePlayer_2,WavePlayer_3], properties {weights [1,1,1]}}
 {id=WavePlayer_1, properties {bLooping=true, soundWaveAssetPtr=&/Fire_Torch_01}}
 {id=WavePlayer_2, properties {bLooping=true, soundWaveAssetPtr=&/Fire_Torch_02}}
 {id=WavePlayer_3, properties {bLooping=true, soundWaveAssetPtr=&/Fire_Torch_03}}
]
properties {
 attenuationSettings=/Game/Sounds/Attenuation/Fort_TorchLoop_Focus_Cone_Att
 bBypassVolumeScaleForPriority=true
 bOverrideConcurrency=true
 maxDistance=1180
 priority=0
 soundClassObject=/Game/Sounds/FortSoundClassesAndModes/Fort_Ambient_NonLoop
 volumeMultiplier=0.6
 attenuationOverrides {attenuationShape=Cone, attenuationShapeExtents {x=171.3182}}
 concurrencyOverrides {maxCount=3, voiceStealReleaseTime=3}
}
