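# Ambient sound cue for a looping urn fire.
# A Random node picks one of three equally weighted, looping urn fire waves, and a Modulator slightly varies its pitch.
# Spatialization uses the shared torch attenuation asset, as the stored attenuation overrides are not switched on, while a concurrency override limits instances per owner.
# It is routed to the non-loop ambient sound class despite its looping waves.
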SoundCue /Game/Sounds/Trans_AmbientLoops/Fire/Fire_UrnNew_Circle_Cue
root=Modulator_0
nodes [
 {id=Modulator_0, children [Random_1], properties {pitchMax=1.1, pitchMin=0.9}}
 {id=Random_1, children [WavePlayer_0,WavePlayer_4,WavePlayer_5], properties {weights [1,1,1]}}
 {id=WavePlayer_0, properties {bLooping=true, soundWaveAssetPtr=&/Fire_Urn_01}}
 {id=WavePlayer_4, properties {bLooping=true, soundWaveAssetPtr=&/Fire_Urn_02}}
 {id=WavePlayer_5, properties {bLooping=true, soundWaveAssetPtr=&/Fire_Urn_03}}
]
properties {
 attenuationSettings=&/Fire_Torch_Att
 bOverrideConcurrency=true
 maxDistance=1000
 soundClassObject=/Game/Sounds/FortSoundClassesAndModes/Fort_Ambient_NonLoop
 subtitlePriority=5000
 volumeMultiplier=0.65
 attenuationOverrides {
  bEnableFocusInterpolation=true
  bEnableListenerFocus=true
  bEnableOcclusion=true
  dBAttenuationAtMax=-40
  falloffDistance=512
  focusAzimuth=15
  focusVolumeAttenuation=1.5
  nonFocusAzimuth=45
  nonFocusVolumeAttenuation=0.5
  occlusionInterpolationTime=1
  occlusionLowPassFilterFrequency=10000
  occlusionTraceChannel=ECC_GameTraceChannel11
  occlusionVolumeAttenuation=0.5
  stereoSpread=512
  attenuationShapeExtents=<256 64.5942 90>
 }
 concurrencyOverrides {bLimitToOwner=true, maxCount=3}
}
