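# Ambient sound cue for a fire urn, playing a looping fire bed.
# A Random node picks one of three equally weighted looping urn waves, then a Modulator applies slight pitch variation.
# Inline overrides set cone attenuation with listener focus that changes volume on and off focus, plus a small concurrency cap.
# It is assigned to the Fort_Ambient_NonLoop sound class despite its looping waves.
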SoundCue /Game/Sounds/Trans_AmbientLoops/Fire/Fire_UrnNew_Cue
root=Modulator_0
nodes [
 {id=Modulator_0, children [Random_1], properties {pitchMax=1.1, pitchMin=0.9}}
 {id=Random_1, children [WavePlayer_0,WavePlayer_4,WavePlayer_5], properties {weights [1,1,1]}}
 {id=WavePlayer_0, properties {bLooping=true, soundWaveAssetPtr=&/Fire_Urn_01}}
 {id=WavePlayer_4, properties {bLooping=true, soundWaveAssetPtr=&/Fire_Urn_02}}
 {id=WavePlayer_5, properties {bLooping=true, soundWaveAssetPtr=&/Fire_Urn_03}}
]
properties {
 attenuationSettings=/Game/Sounds/Attenuation/Fort_TorchLoop_Focus_Cone_Att
 bOverrideAttenuation=true
 bOverrideConcurrency=true
 maxDistance=1043.1733
 soundClassObject=/Game/Sounds/FortSoundClassesAndModes/Fort_Ambient_NonLoop
 volumeMultiplier=0.8
 attenuationOverrides {
  attenuationShape=Cone
  bEnableFocusInterpolation=true
  bEnableListenerFocus=true
  falloffDistance=871.8552
  focusVolumeAttenuation=1.5
  nonFocusVolumeAttenuation=0.3
  attenuationShapeExtents=<171.3182 64.5942 90>
 }
 concurrencyOverrides {maxCount=3}
}
