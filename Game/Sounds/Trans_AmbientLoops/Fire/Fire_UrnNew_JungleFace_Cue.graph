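# Ambient fire loop cue, named for the jungle face urn, that plays one of three looping urn fire waves chosen at equal weight.
# The random pick feeds a slight pitch modulator, then an enveloper that ramps volume up from silence.
# Inline attenuation overrides the referenced torch-loop asset, adding natural falloff, occlusion and listener focus.
# It has zero priority, owner-limited concurrency and, despite looping, the Fort_Ambient_NonLoop sound class.
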SoundCue /Game/Sounds/Trans_AmbientLoops/Fire/Fire_UrnNew_JungleFace_Cue
root=Enveloper_0
nodes [
 {
  id=Enveloper_0
  children [Modulator_0]
  properties {pitchCurve {editorCurveData {keys [{time=0}]}}, volumeCurve {editorCurveData {keys [<at 0 0>,<at 1 1>]}}}
 }
 {id=Modulator_0, children [Random_1], properties {pitchMax=1.1, pitchMin=0.9}}
 {id=Random_1, children [WavePlayer_0,WavePlayer_4,WavePlayer_5], properties {weights [1,1,1]}}
 {id=WavePlayer_0, properties {bLooping=true, soundWaveAssetPtr=&/Fire_Urn_01}}
 {id=WavePlayer_4, properties {bLooping=true, soundWaveAssetPtr=&/Fire_Urn_02}}
 {id=WavePlayer_5, properties {bLooping=true, soundWaveAssetPtr=&/Fire_Urn_03}}
]
properties {
 attenuationSettings=/Game/Sounds/Attenuation/Fort_TorchLoop_Focus_Cone_Att
 bBypassVolumeScaleForPriority=true
 bOverrideAttenuation=true
 bOverrideConcurrency=true
 maxDistance=3200
 priority=0
 soundClassObject=/Game/Sounds/FortSoundClassesAndModes/Fort_Ambient_NonLoop
 subtitlePriority=5000
 volumeMultiplier=0.6
 attenuationOverrides {
  bEnableFocusInterpolation=true
  bEnableListenerFocus=true
  bEnableOcclusion=true
  dBAttenuationAtMax=-40
  distanceAlgorithm=NaturalSound
  falloffDistance=2000
  focusAttackInterpSpeed=2
  focusAzimuth=15
  focusReleaseInterpSpeed=2
  focusVolumeAttenuation=1.5
  nonFocusAzimuth=45
  nonFocusVolumeAttenuation=0.7
  occlusionInterpolationTime=0.5
  occlusionLowPassFilterFrequency=10000
  occlusionTraceChannel=ECC_GameTraceChannel11
  occlusionVolumeAttenuation=0.5
  stereoSpread=2000
  attenuationShapeExtents=<200 64.5942 90>
 }
 concurrencyOverrides {bLimitToOwner=true, maxCount=2}
}
