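# Looping ambient fireplace cue for Athena, blending distant and close fire layers by distance.
# The distant branch mixes Fireplace005 with the Winterfest far loop, which plays only on the first two of four quality slots.
# The close branch plays the Winterfest close loop; each branch has its own attenuation before mixing.
# A root modulator randomizes pitch slightly, and a concurrency override caps instances at two.
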SoundCue /Game/Sounds/Trans_AmbientLoops/Fire/Fireplace_Loop_Athena_Cue
root=Modulator_0
nodes [
 {
  id=Attenuation_0
  children [Mixer_2]
  properties {
   attenuationSettings=/Game/Sounds/Attenuation/Fort_Fireplace_Athena_Distant_Att
   attenuationOverrides {
    distanceAlgorithm=Custom
    falloffDistance=0
    attenuationShapeExtents {x=0}
    customAttenuationCurve {editorCurveData {keys [{interpMode=RCIM_Cubic, tangentMode=RCTM_Break, time=0, value=1}, <at 1 1>]}}
   }
  }
 }
 {
  id=Attenuation_3
  children [WavePlayer_0]
  properties {
   attenuationSettings=/Game/Sounds/Attenuation/Fort_Fireplace_Athena_Close_Att
   attenuationOverrides {
    distanceAlgorithm=Custom
    falloffDistance=0
    attenuationShapeExtents {x=0}
    customAttenuationCurve {editorCurveData {keys [{interpMode=RCIM_Cubic, tangentMode=RCTM_Break, time=0, value=1}, {interpMode=RCIM_Cubic, tangentMode=RCTM_Break, time=1, value=0}]}}
   }
  }
 }
 {id=Mixer_1, children [Attenuation_0, Attenuation_3], properties {inputVolume [2,0.7]}}
 {id=Mixer_2, children [WavePlayer_8, QualityLevel_0], properties {inputVolume [1,0.4]}}
 {id=Modulator_0, children [Mixer_1], properties {pitchMax=1.1, pitchMin=0.9}}
 {id=QualityLevel_0, children [WavePlayer_1,WavePlayer_1,null,null]}
 {id=WavePlayer_0, properties {bLooping=true, soundWaveAssetPtr=/Game/Athena/Sounds/Events/Winterfest/Winterfest_Fireplace_Close_Loop}}
 {id=WavePlayer_1, properties {bLooping=true, soundWaveAssetPtr=/Game/Athena/Sounds/Events/Winterfest/Winterfest_Fireplace_Far_Loop}}
 {id=WavePlayer_8, properties {bLooping=true, soundWaveAssetPtr=&/Fireplace005_Loop}}
]
properties {
 bOverrideConcurrency=true
 maxDistance=1200
 soundClassObject=&Fort_Ambient_Athena
 volumeMultiplier=0.501
 concurrencyOverrides {maxCount=2, volumeScaleMode=Distance}
}
