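# Sound cue for an ambient fireplace loop.
# A Random node picks one of seven equally weighted looping fireplace waves, with repeats allowed, and a Modulator varies its pitch slightly.
# Playback uses fireplace attenuation and an ambient tile-mix sound class at reduced volume, with a concurrency override capping it at two simultaneous instances.
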SoundCue /Game/Sounds/Trans_AmbientLoops/Fire/Fireplace_Loop_Cue
root=Modulator_0
nodes [
 {id=Modulator_0, children [Random_0], properties {pitchMax=1.1, pitchMin=0.9}}
 {
  id=Random_0
  children [WavePlayer_0,WavePlayer_1,WavePlayer_2,WavePlayer_3,WavePlayer_4,WavePlayer_5,WavePlayer_6]
  properties {bRandomizeWithoutReplacement=false, weights [1,1,1,1,1,1,1]}
 }
 {id=WavePlayer_0, properties {bLooping=true, soundWaveAssetPtr=&/Fireplace007_Loop}}
 {id=WavePlayer_1, properties {bLooping=true, soundWaveAssetPtr=&/Fireplace006_Loop}}
 {id=WavePlayer_2, properties {bLooping=true, soundWaveAssetPtr=&/Fireplace005_Loop}}
 {id=WavePlayer_3, properties {bLooping=true, soundWaveAssetPtr=&/Fireplace004_Loop}}
 {id=WavePlayer_4, properties {bLooping=true, soundWaveAssetPtr=&/Fireplace003_Loop}}
 {id=WavePlayer_5, properties {bLooping=true, soundWaveAssetPtr=&/Fireplace002_Loop}}
 {id=WavePlayer_6, properties {bLooping=true, soundWaveAssetPtr=&/Fireplace001_Loop}}
]
properties {
 attenuationSettings=/Game/Sounds/Attenuation/Trans_VFX_Fireplace
 bOverrideConcurrency=true
 maxDistance=1100
 soundClassObject=/Game/Sounds/FortSoundClassesAndModes/Fort_Ambient_TileMix
 volumeMultiplier=0.5
 concurrencyOverrides {maxCount=2}
}
